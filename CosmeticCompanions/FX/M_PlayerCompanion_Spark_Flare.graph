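# Additive unlit Niagara sprite material for the player companion's spark flare, combining a soft corona, spinning spikes and a hot center point.
# A sphere mask indexes the corona gradient texture, while two counter-rotating spikeball samples form rays, all tinted by particle color.
# Dynamic parameters drive corona brightness, size and thickness, plus flicker built from layered sines.
# Deferred shading remaps rays and center to HDR; depth and near-camera fades soften opacity.
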Material /CosmeticCompanions/FX/M_PlayerCompanion_Spark_Flare
graph {
 properties {
  bAutomaticallySetUsageInEditor=false
  bUsedWithNiagaraSprites=true
  blendMode=BLEND_Additive
  shadingModel=MSM_Unlit
  usedShadingModels=Unlit
 }
 nodes [
  n1 = DynamicParameter() {defaultValue=<rgba 0.5 1 0.25 1>, paramNames [CoronaBrightness, CoronaSize, CoronaThickness, FlickerIntensity]}
  n2 = Multiply(A=@n1.CoronaThickness, B=NamedRerouteUsage())
  n3 = Saturate(None=@n2)
  n4 = SphereMask(
   A=TextureCoordinate()
   B=Constant() {r=0.5}
   Radius=Multiply(A=LinearInterpolate(Alpha=@n3) {constA=0.5, constB=0.4}, B=@n1.CoronaSize)
   Hardness=LinearInterpolate(Alpha=@n2) {constB=0.8}
  ) {hardnessPercent=25}
  n5 = Constant2Vector() {r=0.5}
  n6 = Frac(None=Add(A=TextureCoordinate(), B=@n5))
  n7 = LinearInterpolate(Alpha=@n2) {constA=0.55, constB=0.4}
  n8 = Power(Base=Multiply(
    A=TextureSampleParameter2D(UVs=@n4) {parameterName=ColorGradient, texture=/Game/Effects/Fort_Effects/Textures/Grad/T_Halo_Corona_Gradient}.RGB
    B=SphereMask(A=@n6, B=@n5.R, Radius=@n7) {attenuationRadius=0.4, hardnessPercent=25}
   ))
  n9 = ScalarParameter() {defaultValue=0.4, parameterName=BaseCoronaBrightness}
  n10 = ParticleColor()
  n11 = ScalarParameter() {defaultValue=0.25, parameterName=CoronaParticleColorDesaturation}
  n12 = Multiply(A=Multiply(A=Multiply(A=@n8, B=@n9), B=Desaturation(None=@n10.RGB, Fraction=@n11)), B=NamedRerouteUsage())
  n13 = Rotator() {speed=0.5}
  n14 = TextureSample(UVs=Rotator() {speed=-0.35}) {texture=/Game/Effects/Fort_Effects/Textures/Flares/T_Flare_Spikeball}
  n15 = Power(Base=Max(A=TextureSample(UVs=@n13) {texture=&T_Flare_Spikeball}.RGB, B=Multiply(A=@n14.RGB) {constB=0.8})) {constExponent=0.84}
  n16 = ScalarParameter() {defaultValue=0.2, parameterName=OuterBrightness}
  n17 = TextureCoordinate()
  n18 = Constant() {r=0.5}
  n19 = ScalarParameter() {defaultValue=0.06, parameterName=CenterPointSize}
  n20 = Multiply(
   A=Power(Base=SphereMask(A=@n17, B=@n18, Radius=@n19) {attenuationRadius=0.05, hardnessPercent=0}) {constExponent=10}
   B=ScalarParameter() {defaultValue=4, parameterName=CenterPointBrightness}
  )
  n21 = Multiply(A=@n10.RGB, B=Add(A=Multiply(A=@n15, B=@n16) {constB=0.25}, B=@n20))
  n22 = ShadingPathSwitch(
   Default=@n21
   Deferred=MaterialFunctionCall(
    LDR Input=Saturate(None=@n21)
    HDR Tint=@n21
    HDR Intensity Multiplier=ScalarParameter() {defaultValue=2, parameterName=HDRIntensity}
    Desaturate Luminance=Constant()
   ) {materialFunction=&DeriveHDRfromLDR}.RemappedRange
  )
  MaterialFunctionCall(Emissive=Multiply(A=Add(A=@n12, B=@n22), B=NamedRerouteUsage())) {materialFunction=&MF_FX_EmissiveAdjustments}
  MaterialFunctionCall() {materialFunction=&DebugTimeSine}
  n23 = Saturate(None=@n20)
  Multiply(
   A=DepthFade(Opacity=@n10.A, FadeDistance=LinearInterpolate(Alpha=@n23) {constA=30, constB=5})
   B=MaterialFunctionCall(Distance=ScalarParameter() {defaultValue=300, parameterName=NearCamFadeDistance}) {materialFunction=&Near_Camera_Fade}.Result
  )
  NamedRerouteDeclaration(None=@n1.CoronaBrightness) {name=CoronaBrightness, nodeColor=<rgba 0 0.4141 1 1>}
  n24 = Time()
  n25 = Multiply(A=Sine(None=@n24) {period=0.54}, B=Add(A=Sine(None=@n24) {period=0.25}, B=Sine(None=@n24) {period=0.33}))
  n26 = Multiply(A=Sine(None=@n24) {period=0.182}) {constB=0.5}
  n27 = Multiply(A=Clamp(None=Add(A=@n25, B=@n26)) {minDefault=-1}) {constB=0.5}
  NamedRerouteDeclaration(None=Add(A=Multiply(A=@n27, B=@n1.FlickerIntensity))) {name=Flicker, nodeColor=<rgba 0 0.5781 1 1>}
 ]
}
