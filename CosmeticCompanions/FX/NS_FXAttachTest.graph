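# NS_FXAttachTest is a Niagara system, apparently a test for attaching sprite effects to skeletal mesh bones or sockets.
# A single emitter bursts 13 sprites, each with its own spawn group, and its SpawnRate module is disabled.
# Each update, SkeletalMeshLocation uses the spawn group as bone or socket index to position particles on the system-level skeletal mesh.
# Particles outlive their lifetime, and infinitely lived ones die when the emitter deactivates.
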NiagaraSystem /CosmeticCompanions/FX/NS_FXAttachTest
dependencies {
 usedDataInterfaces [&NiagaraDataInterfaceSkeletalMesh]
 usedDynamicInputs [&SimulationPosition]
 usedModules [
  &self:SystemSpawnScript.SystemScriptSource.SystemScriptGraph.NiagaraNodeAssignment_0.SetVariables_#1
  &self:SystemSpawnScript.SystemScriptSource.SystemScriptGraph.NiagaraNodeAssignment_1.SetVariables_#2
  &SystemState
 ]
}
emitters {NE_Empty {
  emitterUpdateScript [
   &EmitterState(
    Inactive Response=::NewEnumerator1
    Life Cycle Mode=::NewEnumerator1
    Loop Behavior=::NewEnumerator1
    Loop Duration Mode=::NewEnumerator1
    UseLoopDelay=false
   ) {module=EmitterLifeCycle}
   &SpawnBurst_Instantaneous(Spawn Count=13, Use Spawn Probability=false, UseLoopCountLimit=false)
   &SpawnRate(SpawnRate=10, Use Spawn Probability=false) {enabled=false}
  ]
  particleSpawnScript [
   &FN_SetSpawnGroup(Number of Spawn Groups=13)
   &InitializeParticle(Sprite Size Mode=::NewEnumerator3, Uniform Sprite Size=15, UsePositionOffset=false, Write Material Random=true) {module=InitializeParticle001}
  ]
  particleUpdateScript [
   &ParticleState(Kill Particles When Lifetime Has Elapsed=false, Let Infinitely Lived Particles Die When Emitter Deactivates=true)
   &SkeletalMeshLocation(
    Bone / Socket Index=$Particles.SpawnGroup
    MaskSpawnGroup=false
    Mesh Sampling Type=::NewEnumerator5
    Skeletal Mesh=$System.Skeletal Mesh
    Socket Sampling Mode=::NewEnumerator5
   )
   &SolveForcesAndVelocity(Clamp Velocity=false, Limit Acceleration=false)
  ]
  rendererClasses [&NiagaraSpriteRendererProperties]
  renderers [{rendererClass=&NiagaraSpriteRendererProperties}]
 }}
summary {
 systemName=NS_FXAttachTest
 emitters [{emitterName=NE_Empty, rendererClasses [&NiagaraSpriteRendererProperties]}]
 userVariables []
}
