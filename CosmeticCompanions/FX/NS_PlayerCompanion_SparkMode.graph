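# Niagara system for the player companion's spark mode: a glowing, color-tinted spark that accompanies the player pawn.
# Spark lerps between two system-computed positions by User.StateChange, adds waveform wobble and a local offset, and renders sprites plus a light.
# A Pulse value switched by User.Returning drives size, brightness and light radius, while speed and camera proximity also scale the sprite.
# RadialBlur bursts a brief mesh and sprite that shrinks, fading while drifting toward the pawn-relative spark position.
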NiagaraSystem /CosmeticCompanions/FX/NS_PlayerCompanion_SparkMode
dependencies {
 usedDataInterfaces [
  &NiagaraDataInterfaceCurve
  &NiagaraDataInterfaceArrayFloat
  &NiagaraDataInterfaceArrayFloat3
  /Script/Niagara.NiagaraDataInterfaceActorComponent
  &NiagaraDataInterfaceCamera
 ]
 usedDynamicInputs [
  &SelectFloatFromArray
  &MakeIntFromFloat
  &NormalizeVector
  &Multiply_Vector
  &Subtract_Position
  &Add_Vector
  &Add_Float
  &Multiply_VectorByFloat
  &DistanceBetweenPositions
  &SelectVectorFromArray
  &ReturnCameraPosition
 ]
 usedModules [
  &self:SystemSpawnScript.SystemScriptSource.SystemScriptGraph.NiagaraNodeAssignment_7.SetVariables_#1
  &self:SystemSpawnScript.SystemScriptSource.SystemScriptGraph.NiagaraNodeAssignment_8.SetVariables_#2
  &self:SystemSpawnScript.SystemScriptSource.SystemScriptGraph.NiagaraNodeAssignment_10.SetVariables_#3
  &SystemState
  &Timeline
  &self:GetPawnPositionAndVelocity
  &self:SystemSpawnScript.SystemScriptSource.SystemScriptGraph.NiagaraNodeAssignment_0.SetVariables_#4
  &self:SystemSpawnScript.SystemScriptSource.SystemScriptGraph.NiagaraNodeAssignment_9.SetVariables_#5
  &InterpolateOverTime
  &self:SystemSpawnScript.SystemScriptSource.SystemScriptGraph.NiagaraNodeAssignment_5.SetVariables_#6
  &self:SystemSpawnScript.SystemScriptSource.SystemScriptGraph.NiagaraNodeAssignment_2.SetVariables_#7
  &self:SystemSpawnScript.SystemScriptSource.SystemScriptGraph.NiagaraNodeAssignment_6.SetVariables_#8
  &self:SystemSpawnScript.SystemScriptSource.SystemScriptGraph.NiagaraNodeAssignment_3.SetVariables_#9
  &self:SystemSpawnScript.SystemScriptSource.SystemScriptGraph.NiagaraNodeAssignment_4.SetVariables_#10
 ]
}
emitters {
 RadialBlur {
  emitterUpdateScript [
   &EmitterState(Life Cycle Mode=::NewEnumerator1, Loop Behavior=::NewEnumerator1, Loop Delay=0.1, Loop Duration Mode=::NewEnumerator1, UseLoopDelay=false)
   &SpawnBurst_Instantaneous(Spawn Count=1, Use Spawn Probability=false)
  ]
  particleSpawnScript [
   &InitializeParticle(
    Color=&MultiplyLinearColor(Linear Color=$System.BaseSparkColor, Scale Factor=<5 5 5 3>)
    Lifetime=0.3
    Mesh Scale Mode=::NewEnumerator3
    Mesh Uniform Scale=3
    Sprite Size Mode=::NewEnumerator3
    Uniform Sprite Size=110
    UsePositionOffset=false
    Write Material Random=true
   )
   &InitialMeshOrientation(Mesh Orientation Mode=::NewEnumerator1, Use Rotation Vector=true) {module=InitialMeshOrientation001}
   &CameraOffset(Camera Offset Amount=45)
  ]
  particleUpdateScript [
   &ParticleState()
   /SetVariables(Particles.Position=&LerpPosition(
     A=$Particles.Initial.Position
     B=$System.SparkPositionPawn
     Alpha=&FloatFromCurve(FloatCurve=DataInterface() {propertyValues {Curve {keys [<at 0.5 0.05 RCIM_Cubic>, {arriveTangent=1.9391, interpMode=RCIM_Cubic, leaveTangent=1.9391, tangentMode=RCTM_User, time=1, value=0.5}]}}})
    )) {isSetParameters=true}
   &ScaleSpriteSize(Uniform Curve Sprite Scale=DataInterface() {propertyValues {Curve {keys [{arriveTangent=-2.1377, interpMode=RCIM_Cubic, time=0, value=1}, {interpMode=RCIM_Cubic, leaveTangent=-0.9635, time=0.7, value=0}]}}})
   &ScaleColor(
    Scale Alpha=&FloatFromCurve(
     FloatCurve=DataInterface() {propertyValues {Curve {keys [
         {arriveTangent=36.2969, interpMode=RCIM_Cubic, leaveTangent=36.2969, tangentMode=RCTM_User, time=0, value=0}
         <at 0.07 1 RCIM_Cubic>
         {arriveTangentWeight=0.31, interpMode=RCIM_Cubic, tangentMode=RCTM_User, time=1, value=0}
        ]}}}
     Scale Curve=&Lerp_Float(A=0, B=0.5, Alpha=$User.StateChange)
    )
    ScaleA=true
    ScaleRGB=false
   )
   &ScaleMeshSize(Scale Factor=&VectorFromFloat(Value=&FloatFromCurve(FloatCurve=DataInterface() {propertyValues {Curve {keys [{arriveTangent=-1.5559, interpMode=RCIM_Cubic, leaveTangent=-1.5559, tangentMode=RCTM_User, time=0, value=1}, <at 1 0 RCIM_Cubic>]}}})))
   &DynamicMaterialParameters(
    Index 0 Param 1=&FloatFromCurve(FloatCurve=DataInterface() {propertyValues {Curve {keys [<at 0 1 RCIM_Cubic>,<at 1 0>]}}})
    Index 3 Param 1=&FloatFromCurve(
     FloatCurve=DataInterface() {propertyValues {Curve {keys [
         {arriveTangent=1.5531, interpMode=RCIM_Cubic, leaveTangent=1.5531, tangentMode=RCTM_User, time=0, value=0.2}
         {tangentMode=RCTM_User, time=1, value=1}
        ]}}}
     Scale Curve=2
    )
    Module.Index0Param0Write=true
    Module.Index0Param1Write=false
    Module.Index0Param2Write=false
    Module.Index0Param3Write=false
    Module.Index3Param0Write=true
    Module.Index3Param1Write=false
    Module.Index3Param2Write=false
    Module.Index3Param3Write=false
    Write Parameter Index 3=true
   ) {module=DynamicMaterialParameters002}
  ]
  rendererClasses [&NiagaraMeshRendererProperties, &NiagaraSpriteRendererProperties]
  renderers [
   {rendererClass=&NiagaraMeshRendererProperties, rendererIndex=0}
   {rendererClass=&NiagaraSpriteRendererProperties, rendererIndex=1}
  ]
 }
 Spark {
  emitterUpdateScript [
   &EmitterState(
    Inactive Response=::NewEnumerator1
    Life Cycle Mode=::NewEnumerator1
    Loop Behavior=::NewEnumerator1
    Loop Duration Mode=::NewEnumerator1
    UseLoopDelay=false
   )
   &SpawnBurst_Instantaneous(Use Spawn Probability=false, UseLoopCountLimit=false)
  ]
  particleSpawnScript [
   &InitializeParticle(
    Color=&MultiplyLinearColor(Linear Color=$System.BaseSparkColor, Scale Factor=<3 3 3 1>)
    Lifetime=0.5
    Sprite Size Mode=::NewEnumerator3
    Uniform Sprite Size=40
    UsePositionOffset=false
    Write Material Random=true
   )
   &CameraOffset(Camera Offset Amount=30)
  ]
  particleUpdateScript [
   &ParticleState(Kill Particles When Lifetime Has Elapsed=false)
   /SetVariables(Particles.Pulse:NiagaraFloat=&MakeCustomFloatFromBool(
     Bool=$User.Returning
     True Float=&FloatFromCurve(
      FloatCurve=DataInterface() {propertyValues {Curve {keys [<at 0 0 RCIM_Cubic>, {arriveTangent=2.3568, interpMode=RCIM_Cubic, leaveTangent=2.3568, tangentMode=RCTM_User, time=1, value=1}]}}}
      CurveIndex=$System.Timeline.InverseNormalizedTime
     )
     False Float=&FloatFromCurve(FloatCurve=DataInterface() {propertyValues {Curve {keys [
          {arriveTangent=10.9729, interpMode=RCIM_Cubic, leaveTangent=10.9729, tangentMode=RCTM_User, time=0, value=0.5}
          {arriveTangent=0.0574, interpMode=RCIM_Cubic, leaveTangent=0.0574, tangentMode=RCTM_User, time=0.1, value=1}
          {arriveTangent=-2.9524, interpMode=RCIM_Cubic, leaveTangent=-2.9524, tangentMode=RCTM_User, time=0.1959, value=0.8819}
          <at 1 0>
         ]}}})
    )) {isSetParameters=true}
   /SetVariables(Particles.Position=&AddVectorToPosition(
     Position=&AddVectorToPosition(
      Position=&Lerp_Vector(
       A=Linked() {linkedVariable {name=System.SparkPosition, type=NiagaraPosition}}
       B=Linked() {linkedVariable {name=System.SparkPositionPawn, type=NiagaraPosition}}
       Alpha=&ClampFloat(Float=$User.StateChange)
      )
      Vector=&MakeVector(
       X=&Waveform(
        Waveform Count=::NewEnumerator1
        "Waveform [1]"=::NewEnumerator0
        "[1] Phase Scale"=0.25
        "[1] Frequency"=1
        "[1] Amplitude Min/Max"=<0 10>
        "[2] Average"=false
        "Waveform [2]"=::NewEnumerator9
        "[2] Input Position"=0
        "[2] Phase"=$Emitter.Age
        "[2] Phase Scale"=1
        "[2] Compound Frequency"=2
        "[2] Frequency"=0.1
        "[2] Amplitude Min/Max"=<0 5>
        "[2] Scale"=1
        "[2] Bias"=0
        "[2] Exponent"=1
        "[2] Pulse Duration"=0.1
        "[2] Clip Triangle Min/Max"=<0 0>
        "[2] Random Seed"=0
       )
       Y=&Waveform(
        Waveform Count=::NewEnumerator1
        "Waveform [1]"=::NewEnumerator0
        "[1] Phase Scale"=0.25
        "[1] Frequency"=0.3
        "[1] Amplitude Min/Max"=<0 10>
        "[2] Average"=false
        "Waveform [2]"=::NewEnumerator9
        "[2] Input Position"=0
        "[2] Phase"=$Emitter.Age
        "[2] Phase Scale"=1
        "[2] Compound Frequency"=2
        "[2] Frequency"=0.2
        "[2] Amplitude Min/Max"=<0 5>
        "[2] Scale"=1
        "[2] Bias"=0
        "[2] Exponent"=1
        "[2] Pulse Duration"=0.1
        "[2] Clip Triangle Min/Max"=<0 0>
        "[2] Random Seed"=0
       )
       Z=&Waveform(
        Waveform Count=::NewEnumerator1
        "Waveform [1]"=::NewEnumerator0
        "[1] Phase Scale"=0.25
        "[1] Frequency"=0.5
        "[1] Amplitude Min/Max"=<0 10>
        "[2] Average"=false
        "Waveform [2]"=::NewEnumerator9
        "[2] Input Position"=0
        "[2] Phase"=$Emitter.Age
        "[2] Phase Scale"=1
        "[2] Compound Frequency"=2
        "[2] Frequency"=0.1
        "[2] Amplitude Min/Max"=<0 5>
        "[2] Scale"=1
        "[2] Bias"=0
        "[2] Exponent"=1
        "[2] Pulse Duration"=0.1
        "[2] Clip Triangle Min/Max"=<0 0>
        "[2] Random Seed"=0
       )
      )
     )
     Vector=$User.SparkLocalOffset
    )) {isSetParameters=true}
   &CalculateAccurateVelocity(Set Intrinsic Variables=false)
   /SetVariables(Particles.Velocity2:Vector3f=$Particles.CalculateAccurateVelocity.Velocity) {isSetParameters=true}
   /SetVariables(Particles.SpeedScale:NiagaraFloat=&FloatFromCurve(
     FloatCurve=DataInterface() {propertyValues {Curve {keys [<at 400 1 RCIM_Cubic>,<at 600 0>]}}}
     CurveIndex=&VectorLength(Vector=$Particles.Velocity2)
    )) {isSetParameters=true}
   &ScaleSpriteSize(
    Scale Sprite Size Mode=::NewEnumerator0
    Uniform Scale Factor=&Multiply_Float(
     A=&Lerp_Float(A=&Lerp_Float(A=0.6, B=1, Alpha=$Particles.SpeedScale), B=3, Alpha=$Particles.Pulse)
     B=$System.NearCameraShrink
    )
   )
   &DynamicMaterialParameters(
    Index 0 Param 1=&Lerp_Float(A=0.5, B=2, Alpha=$Particles.Pulse)
    Index 0 Param 2=&FloatFromCurve(
     FloatCurve=DataInterface() {propertyValues {Curve {keys [
         {arriveTangent=5.4759, interpMode=RCIM_Cubic, time=0, value=0}
         {arriveTangent=0.1804, interpMode=RCIM_Cubic, leaveTangent=0.1804, tangentMode=RCTM_User, time=0.35, value=1}
         {arriveTangent=0.1086, interpMode=RCIM_Cubic, leaveTangent=0.1086, tangentMode=RCTM_User, time=0.8, value=0.7}
         <at 1 0.8 RCIM_Cubic>
        ]}}}
     Scale Curve=&Lerp_Float(A=0.2, B=1, Alpha=$Particles.SpeedScale)
    )
    Index 0 Param 3=&FloatFromCurve(FloatCurve=DataInterface() {propertyValues {Curve {keys [
         {arriveTangent=2.4929, interpMode=RCIM_Cubic, leaveTangent=2.4929, tangentMode=RCTM_User, time=0, value=0.25}
         {arriveTangent=-3.0062, interpMode=RCIM_Cubic, leaveTangent=-3.0062, tangentMode=RCTM_User, time=0.25, value=0.85}
         {arriveTangent=0.7556, interpMode=RCIM_Cubic, leaveTangent=0.7556, tangentMode=RCTM_User, time=0.8, value=0}
         <at 1 0.25 RCIM_Cubic>
        ]}}})
    Index 0 Param 4=&FloatFromCurve(
     FloatCurve=DataInterface() {propertyValues {Curve {keys [{interpMode=RCIM_Cubic, tangentMode=RCTM_User, time=0.5, value=0}, {interpMode=RCIM_Cubic, tangentMode=RCTM_User, time=1, value=1}]}}}
     Scale Curve=0.4
    )
    Index 3 Param 1=2.5
    Module.Index0Param0Write=false
    Module.Index0Param1Write=true
    Module.Index0Param2Write=true
    Module.Index0Param3Write=true
    Module.Index3Param0Write=true
    Module.Index3Param1Write=false
    Module.Index3Param2Write=false
    Module.Index3Param3Write=false
    Write Parameter Index 3=true
   )
   &ScaleColor(
    Scale Alpha=0.85
    Scale RGB=&VectorFromFloat(Value=&Lerp_Float(A=0.8, B=2, Alpha=$Particles.Pulse))
    ScaleA=true
    ScaleRGB=true
   )
   /SetVariables(
    Particles.LightColor=&MultiplyLinearColor(
     Linear Color=$Particles.Color
     Scale Factor=&MakeVector4_FromLinearColor(LinearColor=<rgba 2.8906 3.6006 5 1>)
    )
    Particles.LightRadius=&Lerp_Float(A=50, B=100, Alpha=$Particles.Pulse)
   ) {isSetParameters=true}
   /SetVariables(Particles.DarkenColor:LinearColor=&MultiplyLinearColor(
     Linear Color=<rgba 0.0033 0.0023 0.02 0.4>
     Scale Factor=&MakeFloat4FromVectorAndFloat(
      "Vector (RGB)"=<1 1 1>
      "Float (Alpha)"=&MakeFloatFromLinearColor(LinearColor=$Particles.Color, Channel:ENiagaraLinearColor_Channels=::NewEnumerator0)
     )
    )) {isSetParameters=true}
  ]
  rendererClasses [&NiagaraSpriteRendererProperties, &NiagaraLightRendererProperties]
  renderers [
   {rendererClass=&NiagaraSpriteRendererProperties}
   {rendererClass=&NiagaraLightRendererProperties, rendererIndex=1}
   {rendererClass=&NiagaraSpriteRendererProperties, rendererIndex=2}
  ]
 }
}
summary {
 systemName=NS_PlayerCompanion_SparkMode
 emitters [
  {emitterName=Spark, rendererClasses [&NiagaraSpriteRendererProperties, &NiagaraLightRendererProperties]}
  {emitterName=RadialBlur, rendererClasses [&NiagaraMeshRendererProperties, &NiagaraSpriteRendererProperties]}
 ]
 userVariables [
  User.bIsFrontEnd {type=NiagaraBool}
  User.bIsFrontEndPreview {type=NiagaraBool}
  User.PlayerPawn {
   type=Object
   description=""
   defaultValue {struct=NiagaraExt_VariableValue_Object, value {object=null, objectClass=&Object}}
  }
  User.Returning {type=NiagaraBool}
  User.SparkColor {type=LinearColor, description="", defaultValue {struct=LinearColor, value=<rgba 0.2317 0.3675 1 1>}}
  User.SparkLocalOffset {type=Vector3f, description="", defaultValue {struct=Vector3f, value=<0 0 0>}}
  User.StateChange {type=NiagaraFloat, defaultValue=0, description=""}
 ]
}
