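# Parent material for cosmetic companion eyes on skeletal meshes, with morph target and mesh deformer usage enabled.
# It is a thin Material Attributes wrapper with no inline shading math: MF_animal_eyes_2025 builds the eye surface and feeds MF_TechArtOverrides.
# That result enters CharacterDeRez_Transition as the last function, adding the shared character de-rez transition.
# Shading model comes from the material expression, with Default Lit as the one used.
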Material /CosmeticCompanions/Materials/M_Companion_Eyes_Parent
graph {
 properties {
  bUseMaterialAttributes=true
  bUsedWithMeshDeformer=true
  bUsedWithMorphTargets=true
  bUsedWithSkeletalMesh=true
  pixelDepthOffsetMode=PDOM_Legacy
  shadingModel=MSM_FromMaterialExpression
  usedShadingModels=Default Lit
 }
 nodes [MaterialFunctionCall!(
   Fresnel
   Material=MaterialFunctionCall(In=MaterialFunctionCall() {materialFunction=/Game/Global/Materials/Character_Parent_2023/Eye_Parent_2023/MF_animal_eyes_2025}.Result) {materialFunction=&MF_TechArtOverrides}.Result
  ) {materialFunction=/Game/Packages/Fortress_SharedMaterials/Base_Material_Functions/CharacterDeRez_Transition}]
}
