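# Material function that customizes cosmetic companion textiles by modifying the base color and normal of incoming material attributes.
# A packed mask blends primary, secondary and tertiary colors as a multiplied tint.
# Mask-selected zones receive a tiled, rotatable diffuse pattern and an angle-corrected detail normal, with a scalable red-channel logo overlaid via UV2.
# The FlattenNormal result and the M and TechArtMask inputs are unconnected, and logo position and scale each reuse their X value.
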MaterialFunction /CosmeticCompanions/Materials/MaterialFunctions/MF_CompanionsTextiles
graph {nodes [
  FunctionInput() {inputName=M, sortPriority=1}
  FunctionInput() {inputName=TechArtMask, sortPriority=2}
  FunctionOutput(None=NamedRerouteUsage()) {outputName=MA}
  n1 = TextureSampleParameter2D(UVs=Multiply(
    A=TextureCoordinate() {coordinateIndex=1}
    B=ScalarParameter() {defaultValue=12, group=01_Parameters, parameterName=PatternDetailNormal_Tiling, sliderMax=24, sortPriority=2}
   )) {
   group=00_Texture Inputs
   parameterName=Pattern_DetailNormal
   samplerSource=SSM_Wrap_WorldGroupSettings
   samplerType=SAMPLERTYPE_Normal
   sortPriority=1
   texture=&FlatNormalBlue
  }
  n2 = ScalarParameter() {
   enumeration=/Game/NeverCook/EditorUtilities/AssetDataUtilityTool/Data/ADUT_TypeEnum
   group=01_Parameters
   parameterName=Pattern_DetailNormal_Flatness
   sliderMax=1
  }
  MaterialFunctionCall(Normal=@n1.RGB, Flatness=@n2) {materialFunction=&FlattenNormal}
  NamedRerouteDeclaration(None=FunctionInput() {inputName=MA, inputType=FunctionInput_MaterialAttributes}) {name=MA_In, nodeColor=<rgba 0.0028 1 0.3935 1>}
  n3 = GetMaterialAttributes(None=NamedRerouteUsage()) {attributeGetTypes [#1, #2]}
  NamedRerouteDeclaration(None=SetMaterialAttributes!(MaterialAttributes=@n3.MaterialAttributes, BaseColor=NamedRerouteUsage(), Normal=NamedRerouteUsage()) {attributeSetTypes [#1, #2]}) {name=MA_Out, nodeColor=<rgba 0.3828 1 0 1>}
  NamedRerouteDeclaration(None=@n3.BaseColor) {name=BaseColor_In, nodeColor=<rgba 1 0.0233 0 1>}
  n4 = NamedRerouteUsage()
  n5 = TextureSampleParameter2D() {
   group=00_Texture Inputs
   parameterName=TextileMask_FX1
   samplerType=SAMPLERTYPE_Masks
   sortPriority=1
   texture=&T_BlackMask
   channelNames {a="", b=Trim, g=Secondary, r=Primary}
  }
  n6 = LinearInterpolate(
   A=LinearInterpolate(B=VectorParameter() {group=01_Parameters, parameterName=Color_Primary, sortPriority=3, defaultValue=<rgba 1 0 0 1>}.RGB, Alpha=@n5.Primary)
   B=VectorParameter() {group=01_Parameters, parameterName=Color_Secondary, sortPriority=3, defaultValue=<rgba 0 1 0 1>}.RGB
   Alpha=@n5.Secondary
  )
  n7 = VectorParameter() {group=01_Parameters, parameterName=Color_Tertiary, sortPriority=3, defaultValue=<rgba 0 0 1 1>}
  n8 = LinearInterpolate(
   A=@n4
   B=Multiply(A=@n4, B=LinearInterpolate(A=@n6, B=@n7.RGB, Alpha=@n5.Trim))
   Alpha=ScalarParameter() {group=01_Parameters, parameterName=ColorCustomization_Alpha, sliderMax=1, sortPriority=2}
  )
  n9 = TextureCoordinate() {coordinateIndex=1}
  n10 = ScalarParameter() {defaultValue=12, group=01_Parameters, parameterName=Pattern_Tiling, sliderMax=24, sortPriority=2}
  n11 = TextureSampleParameter2D(UVs=Rotator(
    Coordinate=Multiply(A=@n9, B=@n10)
    Time=ScalarParameter() {group=01_Parameters, parameterName=Pattern_Rotation, sliderMax=1, sliderMin=-1, sortPriority=2}
   ) {speed=3.1415}) {
   constCoordinate=1
   group=00_Texture Inputs
   parameterName=Pattern_Diffuse
   samplerSource=SSM_Wrap_WorldGroupSettings
   sortPriority=1
   texture=&T_BlacksRGB
  }
  n12 = Saturate(None=DotProduct(A=NamedRerouteUsage(), B=VectorParameter() {group=01_Parameters, parameterName=Pattern_Zones, sortPriority=3}.RGB))
  n13 = VectorParameter() {
   group=01_Parameters
   parameterName=Logo_PositionScale
   channelNames {a=Scale Y, b=Scale X, g=Pos Y, r=Pos X}
   defaultValue=<rgba 0 0 0.5 0.5>
  }
  n14 = MaterialFunctionCall(
   UVs=Add(A=TextureCoordinate() {coordinateIndex=2}, B=AppendVector(A=@n13.Pos X, B=@n13.Pos X))
   Texture Scale=AppendVector(A=@n13.Scale X, B=@n13.Scale X)
  ) {materialFunction=&ScaleUVsByCenter}
  n15 = TextureSampleParameter2D(UVs=Rotator(Coordinate=@n14.UVs, Time=ScalarParameter() {group=01_Parameters, parameterName=Logo_Rotation, sliderMax=1, sliderMin=-1, sortPriority=2}) {speed=3.1415}) {
   group=00_Texture Inputs
   parameterName=Logo
   samplerSource=SSM_Clamp_WorldGroupSettings
   sortPriority=1
   texture=&T_BlacksRGB
  }
  NamedRerouteDeclaration(None=LinearInterpolate(A=LinearInterpolate(A=@n8, B=@n11.RGB, Alpha=@n12), B=@n15.R, Alpha=Multiply(A=@n15.R, B=@n14.0-1 mask))) {name=BaseColor_Out, nodeColor=<rgba 1 0.2539 0 1>}
  NamedRerouteDeclaration(None=@n3.Normal) {name=Normal_In, nodeColor=<rgba 0 0.8297 1 1>}
  n16 = NamedRerouteUsage()
  n17 = DotProduct(A=NamedRerouteUsage(), B=VectorParameter() {group=01_Parameters, parameterName=PatternDetailNormal_Zones, sortPriority=3}.RGB)
  NamedRerouteDeclaration(None=LinearInterpolate(
    A=@n16
    B=MaterialFunctionCall(BaseNormal=@n16, AdditionalNormal=@n1.RGB) {materialFunction=&BlendAngleCorrectedNormals}.Result
    Alpha=Saturate(None=@n17)
   )) {name=Normal_Out, nodeColor=<rgba 0 0.1808 1 1>}
  NamedRerouteDeclaration(None=@n5.RGB) {name=TextileMask_FX1, nodeColor=<rgba 1 0 0.0957 1>}
 ]}
