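# Parent Default Lit material for textile surfaces on skeletal-mesh cosmetic companions, supporting morph targets and mesh deformers.
# Uses material attributes, passing the shared 2023 character PreFX function's result through MF_CompanionsTextiles into the matching PostFX function.
# PreFX also feeds PostFX directly for mask, pre-skinned position, distance blend and Fresnel inputs.
# A TechArt cloth fuzz function is present but shows no wired connections.
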Material /CosmeticCompanions/Materials/M_Companion_Textile_Parent_2025
graph {
 properties {
  bAutomaticallySetUsageInEditor=false
  bUseMaterialAttributes=true
  bUsedWithMeshDeformer=true
  bUsedWithMorphTargets=true
  bUsedWithSkeletalMesh=true
  bUsedWithStaticLighting=true
  usedShadingModels=Default Lit
 }
 nodes [
  n1 = MaterialFunctionCall() {materialFunction=&MF_CharacterParent_2023_PreFX}
  MaterialFunctionCall(
   Material=MaterialFunctionCall!(MA=@n1.Result, M=@n1.Result, TechArtMask=@n1.Result) {materialFunction=/CosmeticCompanions/Materials/MaterialFunctions/MF_CompanionsTextiles}.MA
   M=@n1.M
   Pre Skinned Local Position=@n1.M
   Distance Blend=@n1.M
   Fresnel=@n1.M
  ) {materialFunction=&MF_CharacterParent_2023_PostFX}
  MaterialFunctionCall!(Material, Ao/Gmap/Skin, Fresnel, Alpha) {materialFunction=/Game/Global/MaterialFunctions/TechArt/MF_ClothFuzz_TechArt}
 ]
}
