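# Skeletal mesh material for Sidekick cosmetic companions that adds an optional thin-film iridescence effect.
# It wraps the shared character parent PreFX and PostFX functions and includes a companion custom color function.
# A warped, scaled coordinate, apparently Fresnel-driven and optionally noise-offset, samples a spectrum texture added into base color.
# An exponent and roughness influence shape the film, which a Tech Art mask channel blends in behind a static switch.
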Material /CosmeticCompanions/Materials/M_SideKick_ThinFilm
graph {
 properties {
  bAutomaticallySetUsageInEditor=false
  bUseMaterialAttributes=true
  bUsedWithMeshDeformer=true
  bUsedWithMorphTargets=true
  bUsedWithSkeletalMesh=true
  bUsedWithStaticLighting=true
  shadingModel=MSM_FromMaterialExpression
  usedShadingModels=Default Lit | Subsurface | Clear Coat | Subsurface Profile | Cloth
 }
 nodes [
  n1 = MaterialFunctionCall() {materialFunction=&MF_CharacterParent_2023_PreFX}
  MaterialFunctionCall(Material=NamedRerouteUsage(), M=@n1.M, Pre Skinned Local Position=@n1.M, Distance Blend=@n1.M, Fresnel=@n1.M) {materialFunction=&MF_CharacterParent_2023_PostFX}
  n2 = GetMaterialAttributes(None=NamedRerouteUsage()) {attributeGetTypes [#1, #2]}
  NamedRerouteDeclaration(None=@n2.MaterialAttributes) {name=Material Attributes, nodeColor=<rgba 0.0781 0 1 1>}
  NamedRerouteDeclaration(None=Fresnel(ExponentIn=Constant() {r=1}) {baseReflectFraction=0}) {name=Fresnel, nodeColor=<rgba 1 0.2109 0 1>}
  NamedRerouteDeclaration(None=MaterialFunctionCall!(Material=@n1.Result, Tech Art Mask=@n1.Result) {materialFunction=/CosmeticCompanions/Materials/MaterialFunctions/MF_FN_Companion_Custom_Color}.Base Color) {name=OUT_Material, nodeColor=<rgba 0 0.1094 1 1>}
  NamedRerouteDeclaration(None=@n2.Roughness) {name=OUT_Roughness, nodeColor=<rgba 1 0 0.75 1>}
  NamedRerouteDeclaration(None=@n2.Base Color) {name=OUT_BaseColor, nodeColor=<rgba 0 1 0.9297 1>}
  NamedRerouteDeclaration(None=@n1.Tech Art Mask) {name=TechArtMask, nodeColor=<rgba 1 0.6094 0 1>}
  n3 = NamedRerouteUsage()
  n4 = NamedRerouteUsage()
  n5 = TextureSampleParameter2D(UVs=Multiply(A=TextureCoordinate(), B=ScalarParameter() {defaultValue=2, group=06_Thin Film | Side Kick, parameterName=SideKick.UvFilmOffsetTiling})) {
   group=06_Thin Film | Side Kick
   parameterName=SideKick.ThinFilmUVOffsetTexture
   samplerSource=SSM_Wrap_WorldGroupSettings
   texture=/Game/Global/Textures/Default/Noise/3xPerlin_M
  }
  n6 = ChannelMaskParameter(None=@n5.RGB) {group=06_Thin Film | Side Kick, maskChannel=Green, parameterName=SideKick.ThinFilmUVOffsetChannel, defaultValue=<rgba 0 1 0 0>}
  n7 = Multiply(
   A=Multiply(A=@n6.R) {constB=0.1}
   B=ScalarParameter() {defaultValue=0.2, group=06_Thin Film | Side Kick, parameterName=SideKick.ThinFilmUVOffsetIntensity}
  )
  n8 = NamedRerouteUsage()
  n9 = ScalarParameter() {defaultValue=0.5, group=06_Thin Film | Side Kick, parameterName=SideKick.ThinFilmWarp, sortPriority=4}
  n10 = Add(
   A=Multiply(
    A=ScalarParameter() {defaultValue=1, group=06_Thin Film | Side Kick, parameterName=SideKick.ThinFilmScale, sortPriority=5}
    B=Power(Base=@n8, Exp=@n9)
   )
   B=ScalarParameter() {group=06_Thin Film | Side Kick, parameterName=SideKick.ThinFilmOffset}
  )
  n11 = TextureSampleParameter2D(UVs=StaticSwitchParameter(True=Add(A=@n7, B=@n10), False=@n10) {group=06_Thin Film | Side Kick, parameterName=SideKick.UseThinFilmUVOffset, sortPriority=-10}) {
   group=06_Thin Film | Side Kick
   parameterName=SideKick.ThinFilmTexture
   samplerSource=SSM_Wrap_WorldGroupSettings
   sortPriority=-25
   texture=/Game/Global/Textures/Default/Thinfilm/T_ThinFilm_Spectrum_COLOR
  }
  n12 = Multiply(A=@n11.RGB, B=ScalarParameter() {defaultValue=0.5, group=06_Thin Film | Side Kick, parameterName=SideKick.ThinFilmIntensity, sortPriority=2})
  n13 = Power(
   Base=NamedRerouteUsage()
   Exp=ScalarParameter() {defaultValue=2, group=06_Thin Film | Side Kick, parameterName=SideKick.ThinFilmExponent, sortPriority=3}
  )
  n14 = NamedRerouteUsage()
  n15 = LinearInterpolate(
   A=@n13
   B=Multiply(A=LinearInterpolate(Alpha=@n14) {constA=1, constB=0.01}, B=@n13)
   Alpha=ScalarParameter() {group=06_Thin Film | Side Kick, parameterName=SideKick.RoughnessInfluence}
  )
  n16 = SetMaterialAttributes!(MaterialAttributes=NamedRerouteUsage(), Base Color=LinearInterpolate(A=NamedRerouteUsage(), B=Add(A=@n4, B=@n12), Alpha=Saturate(None=@n15))) {attributeSetTypes [#2]}
  n17 = ChannelMaskParameter(None=NamedRerouteUsage()) {group=05_Tech Art Masks, parameterName=ThinFilmMaskChannel}
  NamedRerouteDeclaration(None=StaticSwitchParameter(True=BlendMaterialAttributes(A=@n3, B=@n16, Alpha=@n17.R) {vertexAttributeBlendType=UseA}, False=@n3) {group=06_Thin Film | Side Kick, parameterName=SideKick.UseThinFilm, sortPriority=-30}) {name=IN_Material, nodeColor=<rgba 0.2891 0 1 1>}
 ]
}
